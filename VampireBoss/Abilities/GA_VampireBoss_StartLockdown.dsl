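; Vampire boss gameplay ability that requests a lockdown of its heist base when the boss dies.
; Once playlist data is ready, a scalable toggle either ends the ability without lockdown or, after a short delay, binds pawn and AI-controller death events.
; Both events, plus ability end as fallback, run duplicated, separately DoOnce-guarded logic.
; Each calls lockdown on heist managers matching TagToSearchFor, logging if none is found.
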(event Ability|EventActivateAbility
  (Variables|Default|SetbTryLockdownOnAbilityEnd true)
  (Class|FortAsyncActionCurrentPlaylistReady|PlaylistDataReadyAsync
    (:OnReady
      (if (ScalableFloat|GetValueasBool (Variables|Default|GetShouldTriggerLockdownOnDeath))
        (bind _outfortplayerpawn (CallFunction|GetAvatarasValidFortPlayerPawn)
          (:ValidFortPlayerPawn
            (Utilities|FlowControl|Delay 0.2)
            (EventDispatchers|BindEventtoOnDied _outfortplayerpawn (EventDispatchers|CreateEvent :!event Custom|BossDied))
            (bind _asfort_athena_aibot_controller (Utilities|Casting|CastToFortAthenaAIBotController (Pawn|GetController _outfortplayerpawn)))
            (Death|BindEventtoOnAgentDiedEvent _asfort_athena_aibot_controller (EventDispatchers|CreateEvent :!event Custom|AgentDiedEventFromController))))
        (else
          (Variables|Default|SetbTryLockdownOnAbilityEnd false)
          (Ability|EndAbility))))))

(event Custom|AgentDiedEventFromController
  (params (BotController :type Fort%20Athena%20AIBot%20Controller%20Object%20Reference) (AgentPawn :type Fort%20Pawn%20Object%20Reference))
  (bind _tagtosearchfor (Variables|Default|GetTagToSearchFor))
  (Utilities|FlowControl|DoOnce false
    (:Completed
      (bind _outactors (Actor|GetAllActorsOfClass "/HeistBasesGameplay/Items/BGA_HeistBaseManager.BGA_HeistBaseManager_C"))
      (for _array_element _outactors
        (if (== (Variables|/HeistBasesGameplay/Items/BGA_HeistBaseManager.BGA_HeistBaseManager_C|GetHeistBaseId%20Tag%20Const _array_element) _tagtosearchfor)
          (Variables|Default|SetbFoundHeistManager true)
          (Class|BGAHeistBaseManager|OnRequestLockdownBase _array_element (Utilities|Struct|MakeStructHeistBasesLockdownInitiatedPayload _tagtosearchfor (CallFunction|GetAvatarasFortPlayerPawn) (CallFunction|GetAvatarasFortPlayerPawn)))))
      (if (not (Variables|Default|GetbFoundHeistManager))
        ;[1 0 0.9339] LOG: When lockdown occurs report the number of defenses that have been cached for lockdown behavior
        (Fortnite|FortShippingLog self (Utilities|String|Append "VAMPIRE BOSS DIED. TRIED TO TRIGGER LOCKDOWN FOR: Base (" (GameplayTags|GetDebugStringfromGameplayTag (Variables|Default|GetTagToSearchFor)) ") BUT FAILED TO FIND HEIST MANAGER."))))))

(event Custom|BossDied
  (params (DamagedActor :type Actor%20Object%20Reference) (Damage :type Float%20%28single-precision%29) (InstigatedBy :type Controller%20Object%20Reference) (DamageCauser :type Actor%20Object%20Reference) (HitLocation :type Vector) (FHitComponent :type Primitive%20Component%20Object%20Reference) (BoneName :type Name) (Momentum :type Vector))
  (bind _tagtosearchfor (Variables|Default|GetTagToSearchFor))
  (Utilities|FlowControl|DoOnce false
    (:Completed
      (bind _outactors (Actor|GetAllActorsOfClass "/HeistBasesGameplay/Items/BGA_HeistBaseManager.BGA_HeistBaseManager_C"))
      (for _array_element _outactors
        (if (== (Variables|/HeistBasesGameplay/Items/BGA_HeistBaseManager.BGA_HeistBaseManager_C|GetHeistBaseId%20Tag%20Const _array_element) _tagtosearchfor)
          (Variables|Default|SetbFoundHeistManager true)
          (Class|BGAHeistBaseManager|OnRequestLockdownBase _array_element (Utilities|Struct|MakeStructHeistBasesLockdownInitiatedPayload _tagtosearchfor (CallFunction|GetAvatarasFortPlayerPawn) (CallFunction|GetAvatarasFortPlayerPawn)))))
      (if (not (Variables|Default|GetbFoundHeistManager))
        ;[1 0 0.9339] LOG: When lockdown occurs report the number of defenses that have been cached for lockdown behavior
        (Fortnite|FortShippingLog self (Utilities|String|Append "VAMPIRE BOSS DIED. TRIED TO TRIGGER LOCKDOWN FOR: Base (" (GameplayTags|GetDebugStringfromGameplayTag (Variables|Default|GetTagToSearchFor)) ") BUT FAILED TO FIND HEIST MANAGER."))))))

(event Ability|EventOnEndAbility
  (params (bWasCancelled :type Boolean))
  (bind _tagtosearchfor (Variables|Default|GetTagToSearchFor))
  (if (Variables|Default|GetbTryLockdownOnAbilityEnd)
    (Utilities|FlowControl|DoOnce false
      (:Completed
        (bind _outactors (Actor|GetAllActorsOfClass "/HeistBasesGameplay/Items/BGA_HeistBaseManager.BGA_HeistBaseManager_C"))
        (for _array_element _outactors
          (if (== (Variables|/HeistBasesGameplay/Items/BGA_HeistBaseManager.BGA_HeistBaseManager_C|GetHeistBaseId%20Tag%20Const _array_element) _tagtosearchfor)
            (Variables|Default|SetbFoundHeistManager true)
            (Class|BGAHeistBaseManager|OnRequestLockdownBase _array_element (Utilities|Struct|MakeStructHeistBasesLockdownInitiatedPayload _tagtosearchfor (CallFunction|GetAvatarasFortPlayerPawn) (CallFunction|GetAvatarasFortPlayerPawn)))))
        (if (not (Variables|Default|GetbFoundHeistManager))
          ;[1 0 0.9339] LOG: When lockdown occurs report the number of defenses that have been cached for lockdown behavior
          (Fortnite|FortShippingLog self (Utilities|String|Append "VAMPIRE BOSS DIED. TRIED TO TRIGGER LOCKDOWN FOR: Base (" (GameplayTags|GetDebugStringfromGameplayTag (Variables|Default|GetTagToSearchFor)) ") BUT FAILED TO FIND HEIST MANAGER.")))))))
